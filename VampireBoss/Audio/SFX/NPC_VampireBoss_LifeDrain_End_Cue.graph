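# Sound cue for the end of the Vampire Boss's life drain effect.
# A single wave player plays the LifeDrain_End wave at boosted volume, with no randomization or mixing.
# It routes to the Fort_Gadgets sound class and uses the boss's Land attenuation asset for spatial falloff.
# Playback limits come from the Evil_Boss_Con concurrency set under the NPC library's TinyBalloon audio.
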SoundCue /VampireBoss/Audio/SFX/NPC_VampireBoss_LifeDrain_End_Cue
root=WavePlayer_0
nodes [{id=WavePlayer_0, properties {soundWaveAssetPtr=&/SFX_VampireBoss_LifeDrain_End}}]
properties {
 attenuationSettings=/VampireBoss/Audio/NPC_VampireBoss_Land_Att
 maxDistance=3700
 soundClassObject=&Fort_Gadgets
 volumeMultiplier=1.9953
 attenuationOverrides {falloffDistance=4578.8184, attenuationShapeExtents {x=4135.1602}}
 concurrencySet [/NPCLibrary/Audio/TinyBalloon/Evil_Boss_Con]
}
