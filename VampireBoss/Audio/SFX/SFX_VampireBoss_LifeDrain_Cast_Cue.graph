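# Sound cue for the Vampire Boss's Life Drain cast sound effect.
# A single wave player feeds a mixer that boosts its input, and the cue adds a further volume boost.
# It uses the boss's NPC attenuation asset, the Fort_Gadgets sound class and a raised volume multiplier with set priority.
# Its concurrency set is borrowed from first-person weapon gunfire, alongside a stored max-count override.
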SoundCue /VampireBoss/Audio/SFX/SFX_VampireBoss_LifeDrain_Cast_Cue
root=Mixer_0
nodes [
 {id=Mixer_0, children [WavePlayer_1], properties {inputVolume [1.2]}}
 {id=WavePlayer_1, properties {soundWaveAssetPtr=&/SFX_VampireBoss_LifeDrain_Cast}}
]
properties {
 attenuationSettings=/VampireBoss/Audio/NPC_VampireBoss_Att
 maxDistance=3800
 priority=0.9
 soundClassObject=&Fort_Gadgets
 volumeMultiplier=1.9953
 concurrencyOverrides {maxCount=2}
 concurrencySet [&Weapon_Gunfire_FirstPerson_Conc]
}
