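# Sound cue for the final impact of the Vampire Boss's Life Drain attack.
# A single wave player feeds the output, with boosted volume, close-explosion attenuation and the impacts sound class.
# It references a shared evil-boss concurrency asset from NPCLibrary, alongside concurrency override values that specify one instance per owner.
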SoundCue /VampireBoss/Audio/SFX/SFX_VampireBoss_LifeDrain_FinalImpact_Cue
root=WavePlayer_1
nodes [{id=WavePlayer_1, properties {soundWaveAssetPtr=&/SFX_VampireBoss_LifeDrain_FinalImpact}}]
properties {
 attenuationSettings=/Game/Sounds/Attenuation/Fort_Explosion_Close_StW_Att
 maxDistance=3500
 priority=2
 soundClassObject=&Fort_Impacts_PSM
 volumeMultiplier=1.7
 concurrencyOverrides {bLimitToOwner=true, maxCount=1}
 concurrencySet [/NPCLibrary/Audio/TinyBalloon/Evil_Boss_Con]
}
