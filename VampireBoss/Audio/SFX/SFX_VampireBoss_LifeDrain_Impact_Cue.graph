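# Sound cue for the impact sound of the VampireBoss life drain effect.
# A random node picks one of four equally weighted impact waves and feeds a modulator that adds slight pitch and volume variation.
# It plays at boosted volume through a melee weapon sound class with close-explosion attenuation, and overrides concurrency to cap simultaneous instances.
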SoundCue /VampireBoss/Audio/SFX/SFX_VampireBoss_LifeDrain_Impact_Cue
root=Modulator_1
nodes [
 {id=Modulator_1, children [Random_0], properties {pitchMin=0.99, volumeMax=1.01, volumeMin=0.99}}
 {id=Random_0, children [WavePlayer_0,WavePlayer_4,WavePlayer_5,WavePlayer_6], properties {weights [1,1,1,1]}}
 {id=WavePlayer_0, properties {soundWaveAssetPtr=&/SFX_VampireBoss_LifeDrain_Impact_01}}
 {id=WavePlayer_4, properties {soundWaveAssetPtr=&/SFX_VampireBoss_LifeDrain_Impact_02}}
 {id=WavePlayer_5, properties {soundWaveAssetPtr=&/SFX_VampireBoss_LifeDrain_Impact_03}}
 {id=WavePlayer_6, properties {soundWaveAssetPtr=&/SFX_VampireBoss_LifeDrain_Impact_04}}
]
properties {
 attenuationSettings=/Game/Sounds/Attenuation/Fort_Explosion_Close_StW_Att
 bOverrideConcurrency=true
 maxDistance=3500
 soundClassObject=&Fort_WeaponMelee_PSM
 volumeMultiplier=2
 concurrencyOverrides {maxCount=3}
}
