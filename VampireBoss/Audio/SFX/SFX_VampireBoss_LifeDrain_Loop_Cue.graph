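# Sound cue for the looping life drain effect of the Vampire Boss.
# A single looping wave player feeds all four quality level inputs, so every tier plays the same sound, then passes through a modulator left at default settings.
# It uses a shared NPC glitch attenuation and the gadgets sound class, boosts volume, and overrides concurrency to cap simultaneous instances with volume scaling.
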SoundCue /VampireBoss/Audio/SFX/SFX_VampireBoss_LifeDrain_Loop_Cue
root=Modulator_0
nodes [
 {id=Modulator_0, children [QualityLevel_0]}
 {id=QualityLevel_0, children [WavePlayer_0,WavePlayer_0,WavePlayer_0,WavePlayer_0]}
 {id=WavePlayer_0, properties {bLooping=true, soundWaveAssetPtr=&/SFX_VampireBoss_LifeDrain_Loop}}
]
properties {
 attenuationSettings=/Game/Athena/Sounds/NPCs/NPC_Glitch_Att
 bOverrideConcurrency=true
 maxDistance=1500
 soundClassObject=&Fort_Gadgets
 volumeMultiplier=2.5938
 concurrencyOverrides {maxCount=2, volumeScale=0.5}
}
